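; Blueprint for the GlowDonut CageLight A environment prop, a light that can be switched on and off through two custom events.
; TurnLightOn and TurnLightOff show or hide the SM_Airglow component and set an emissive tint parameter to white or black.
; The tint is written to a dynamic material instance on the static mesh, fetched through the BTechArtGlobalLibrary helper.
; The construction script only calls its parent.
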(event Custom|TurnLightOn
  ;[1 1 1] Toggle Light Visiblity
  (Rendering|SetVisibility (Variables|Default|GetSM_Airglow) true false)
  ;[1 1 1] Light Emissive on/off
  (Rendering|Material|SetVectorParameterValue (Class|BTechArtGlobalLibrary|GetDynamicMaterialInstance (Variables|Debug|GetStaticMeshComponent)) "SFX_Emissive_Tint(Expand)_1" "(R=1.000000,G=1.000000,B=1.000000,A=1.000000)" :!class /Script/Engine.MaterialInstanceDynamic))

(event Custom|TurnLightOff
  ;[1 1 1] Toggle Light Visiblity
  (Rendering|SetVisibility (Variables|Default|GetSM_Airglow) false false)
  ;[1 1 1] Light Emissive on/off
  (Rendering|Material|SetVectorParameterValue (Class|BTechArtGlobalLibrary|GetDynamicMaterialInstance (Variables|Debug|GetStaticMeshComponent)) "SFX_Emissive_Tint(Expand)_1" "(R=0.000000,G=0.000000,B=0.000000,A=1.000000)" :!class /Script/Engine.MaterialInstanceDynamic))

(fn ConstructionScript
  (|CallParentFunction))
